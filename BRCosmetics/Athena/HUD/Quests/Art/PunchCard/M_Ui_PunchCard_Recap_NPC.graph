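# Translucent UI material for an NPC portrait tile on the quest punch card recap.
# A panning, screen-aligned triangle grid color-dodged into a two-color background sits under a radial gradient and the NPC portrait texture, which fades into the background color toward its bottom.
# An angled bottom bar overlays this, and a hover scalar drives portrait zoom, fade, bar tilt, offset and color.
# The shape mask stitches two 9-slice container samples with adjustable edge sharpness.
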Material /BRCosmetics/Athena/HUD/Quests/Art/PunchCard/M_Ui_PunchCard_Recap_NPC
graph {
 properties {blendMode=BLEND_Translucent, materialDomain=MD_UI, pixelDepthOffsetMode=PDOM_Legacy, usedShadingModels=Default Lit}
 nodes [
  n1 = VectorParameter() {group=Colors, parameterName=RGB1, defaultValue=<rgba 0 0.049 0.3646 1>}
  n2 = MaterialFunctionCall(Float2=MaterialFunctionCall() {materialFunction=&ScreenResolution}.Visible Resolution) {materialFunction=&BreakOutFloat2Components}
  n3 = Divide(A=@n2.R, B=Max(A=@n2.G))
  n4 = Constant() {r=1}
  n5 = Rotator(
   Coordinate=Multiply(
    A=AppendVector(A=@n3, B=@n4)
    B=MaterialFunctionCall() {materialFunction=&ScreenAlignedUVs}."X 100%, Y 100%"
   )
   Time=ScalarParameter() {defaultValue=-0.0571, group=Triangles, parameterName=Tri Angle}
  ) {speed=6.2832}
  n6 = AppendVector(A=ScalarParameter() {defaultValue=0.0025, group=Triangles, parameterName=PanSpeed}, B=Constant())
  n7 = TextureSample(UVs=Multiply(A=Panner(Coordinate=@n5, Speed=@n6), B=ScalarParameter() {defaultValue=15, group=Triangles, parameterName=Tri Tile Amount})) {texture=/Game/Global/Textures/Default/Masks_and_Gradients/T-TriangleGrid}
  n8 = ScalarParameter() {defaultValue=0.2, group=Triangles, parameterName=TriangleDensity}
  n9 = MaterialFunctionCall!(
   Blend=LinearInterpolate(A=@n1.RGB, B=VectorParameter() {group=Colors, parameterName=RGB2, defaultValue=<rgba 0 0.4337 1 1>}.RGB, Alpha=Multiply(A=@n7.R, B=@n8))
   Base=@n1.RGB
  ) {materialFunction=/Engine/Functions/Engine_MaterialFunctions03/Blends/Blend_ColorDodge}
  n10 = VectorParameter() {group=Colors, parameterName=RGB2, defaultValue=<rgba 0 0.4337 1 1>}
  n11 = ScalarParameter() {defaultValue=0.5, group=Radial gradient, parameterName=RadialGradient_u, sliderMin=0.5}
  n12 = ScalarParameter() {defaultValue=0.9, group=Radial gradient, parameterName=RadialGradient_V, sliderMin=0.5}
  n13 = Saturate(None=MaterialFunctionCall(
    UVs=MaterialFunctionCall() {materialFunction=&GetUserInterfaceUV}.Normalized UV
    CenterPosition=AppendVector(A=@n11, B=@n12)
    Radius=ScalarParameter() {defaultValue=0.85, group=Radial gradient, parameterName=Radial Gradient Radius_V}
    Density=ScalarParameter() {defaultValue=1, group=Radial gradient, parameterName=Radial Gradient Density}
   ) {materialFunction=&RadialGradientExponential}.RadialGradientExponential)
  n14 = MaterialFunctionCall(Float2=MaterialFunctionCall() {materialFunction=&GetUserInterfaceUV}.Pixel Size) {materialFunction=&BreakOutFloat2Components}
  n15 = Divide(A=@n14.R, B=Max(A=@n14.G))
  n16 = MaterialFunctionCall(Float2=TextureCoordinate() {coordinateIndex=4}) {materialFunction=&BreakOutFloat2Components}
  n17 = Add(
   A=AppendVector(A=Multiply(A=@n15, B=@n16.R), B=@n16.G)
   B=AppendVector(
    A=ScalarParameter() {defaultValue=-0.125, group=Item, parameterName=Itemposition_U}
    B=ScalarParameter() {defaultValue=-0.05, group=Item, parameterName=Itemposition_V}
   )
  )
  n18 = ScalarParameter() {group=Item, parameterName=item rotation}
  n19 = ScalarParameter() {defaultValue=1.2, group=control, parameterName=Zoom}
  n20 = Add(A=@n19, B=ScalarParameter() {defaultValue=0.1, group=control, parameterName=hoverZoom})
  n21 = ScalarParameter() {group=control, parameterName=hover, sliderMax=1}
  n22 = TextureSampleParameter2D(UVs=MaterialFunctionCall(UVs=Rotator(Coordinate=@n17, Time=@n18), Texture Scale=LinearInterpolate(A=@n19, B=@n20, Alpha=@n21) {constA=1, constB=1.01}) {materialFunction=&ScaleUVsByCenter}.UVs) {group=Item, parameterName=NPC-Portrait, samplerSource=SSM_Clamp_WorldGroupSettings, texture=/Game/Athena/HUD/Quests/Art/T_NPC_Default}
  n23 = ComponentMask(None=TextureCoordinate() {coordinateIndex=4}) {g=true}
  n24 = LinearInterpolate(
   A=ScalarParameter() {defaultValue=0.6, group=bottom Bar, parameterName=item_bottom_position}
   Alpha=ScalarParameter() {group=control, parameterName=hover, sliderMax=1}
  ) {constA=1}
  n25 = Saturate(None=Multiply(A=Subtract(A=@n23, B=@n24), B=ScalarParameter() {defaultValue=1, group=bottom Bar, parameterName=item_bottom_strenght}))
  n26 = Rotator(
   Coordinate=TextureCoordinate() {coordinateIndex=4}
   Time=LinearInterpolate(
    A=ScalarParameter() {defaultValue=0.49, group=bottom Bar, parameterName=Bot-Angle}
    B=ScalarParameter() {defaultValue=0.5125, group=bottom Bar, parameterName=Bot-Angle-hover}
    Alpha=ScalarParameter() {group=control, parameterName=hover, sliderMax=1}
   )
  ) {speed=6.2832}
  n27 = ScalarParameter() {defaultValue=0.08, group=bottom Bar, parameterName=Bot-RangeOffset, sliderMax=1}
  n28 = ScalarParameter() {defaultValue=0.085, group=bottom Bar, parameterName=Bot-RangeOffset-hover, sliderMax=1}
  n29 = ScalarParameter() {group=control, parameterName=hover, sliderMax=1}
  n30 = Multiply(
   A=Subtract(A=ComponentMask(None=@n26) {r=true}, B=LinearInterpolate(A=@n27, B=@n28, Alpha=@n29))
   B=ScalarParameter() {defaultValue=128, group=bottom Bar, parameterName=strenght}
  )
  LinearInterpolate(
   A=LinearInterpolate(A=LinearInterpolate(A=@n9.Result, B=@n10.RGB, Alpha=@n13), B=LinearInterpolate(A=@n22.RGB, B=@n10.RGB, Alpha=@n25), Alpha=@n22.A)
   B=LinearInterpolate(
    A=VectorParameter() {group=Colors, parameterName=RGB3, defaultValue=<rgba 0 0.5933 1 1>}.RGB
    Alpha=ScalarParameter() {group=control, parameterName=hover, sliderMax=1}
   )
   Alpha=Saturate(None=OneMinus(None=@n30))
  )
  n31 = MaterialFunctionCall() {materialFunction=&GetUserInterfaceUV}
  n32 = AppendVector(A=Constant() {r=-0.25}, B=Constant())
  n33 = ComponentMask(None=TextureSample(UVs=Add(A=@n31."9-Slice UV (No Tiling)", B=@n32) {constB=-0.5}) {samplerSource=SSM_Clamp_WorldGroupSettings, texture=/Game/Athena/HUD/Quests/Art/T_Quest_Container}.R) {r=true}
  n34 = Add(A=@n31."9-Slice UV (No Tiling)", B=AppendVector(A=Constant() {r=-0.5}, B=Constant()))
  n35 = Saturate(None=Multiply(A=ComponentMask(None=@n34) {r=true}) {constB=64})
  n36 = ComponentMask(None=TextureSample(UVs=@n31."9-Slice UV (No Tiling)", Tex=TextureObject() {texture=&T_Quest_Container}) {texture=&T_Quest_Container}.R) {r=true}
  n37 = OneMinus(None=@n35)
  n38 = Saturate(None=Add(A=Multiply(A=@n33, B=@n35), B=Multiply(A=@n36, B=@n37)))
  n39 = ScalarParameter() {parameterName=EdgeOffSet}
  Saturate(None=Multiply(A=Subtract(A=@n38, B=@n39), B=ScalarParameter() {defaultValue=4, parameterName=EdgeSharpness}))
  TextureCoordinate()
 ]
}
