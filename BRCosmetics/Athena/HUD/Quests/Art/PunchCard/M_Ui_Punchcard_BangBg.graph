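# Translucent UI material for the quest punch card "Bang" background, with a color parameter defaulting to yellow.
# The red channel of the quest container texture is offset and sharpened into an edge-adjustable shape mask.
# Rotated 9-slice UVs, shifted and steeply scaled, form a hard angled cut that multiplies with that mask, clipping the shape to one side.
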Material /BRCosmetics/Athena/HUD/Quests/Art/PunchCard/M_Ui_Punchcard_BangBg
graph {
 properties {blendMode=BLEND_Translucent, materialDomain=MD_UI, pixelDepthOffsetMode=PDOM_Legacy, usedShadingModels=Default Lit}
 nodes [
  n1 = Subtract(
   A=TextureSample(Tex=TextureObject() {texture=/Game/Athena/HUD/Quests/Art/T_Quest_Container}) {texture=&T_Quest_Container}.R
   B=ScalarParameter() {parameterName=EdgeOffSet}
  )
  n2 = ScalarParameter() {defaultValue=4, parameterName=EdgeSharpness}
  n3 = Rotator(
   Coordinate=MaterialFunctionCall() {materialFunction=&GetUserInterfaceUV}.9-Slice UV
   Time=ScalarParameter() {defaultValue=0.35, parameterName=rotation}
  )
  n4 = ScalarParameter() {defaultValue=-0.5, parameterName=position}
  n5 = Multiply(A=ComponentMask(None=Add(A=@n3, B=@n4)) {r=true}, B=ScalarParameter() {defaultValue=64, parameterName=Sharpness})
  Multiply(A=Saturate(None=Multiply(A=@n1, B=@n2)), B=OneMinus(None=Saturate(None=@n5)))
  VectorParameter() {parameterName=Color, defaultValue=<rgba 0.9301 1 0.0097 1>}
 ]
}
